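# Physics asset for the WaningCacti two-handed pickaxe, giving its handle a dangling seven-link chain for physics simulation.
# Two capsules cover the handle, and dyn_chain_01 to 07 form a capsule chain hung from it with limited swing and little or no twist.
# Thigh and calf capsules for both legs link to the handle with fully free angular motion, apparently as body collision proxies for the chain.
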PhysicsAsset /BRCosmetics/Weapons/FORT_Melee/Pickaxe_WaningCacti/Meshes/WaningCacti_Axe_Physics
bodies {
 Melee_TwoHanded_Handle [
  {shapeType=Capsule, length=28.8107, radius=8.9793, center=<-1.246 8.224 10.9575>}
  {
   shapeType=Capsule
   length=22.019
   radius=6.9359
   shapeName=Melee_TwoHanded_Handle_capsule
   center=<2.9212 -7.476 4.9856>
   rotation=<rot -0.0005 0 -175.0702>
  }
 ]
 calf_l [{shapeType=Capsule, length=50, radius=8, center=<-30 -2 0>, rotation=<rot 90 0 0>}]
 calf_r [{shapeType=Capsule, length=50, radius=8, center=<30 2 0>, rotation=<rot 90 0 0>}]
 dyn_chain_01 [{shapeType=Capsule, length=6.6295, radius=1.736, shapeName=dyn_chain_01_capsule, center=<6.4 0 0>, rotation=<rot 90 0 0>}]
 dyn_chain_02 [{shapeType=Capsule, length=5.5176, radius=1.4626, shapeName=dyn_chain_02_capsule, center=<4 0 0>, rotation=<rot 90 0 0>}]
 dyn_chain_03 [{shapeType=Capsule, length=2.8479, radius=1.1145, shapeName=dyn_chain_03_capsule, center=<2.8 0 0>, rotation=<rot 90 0 0>}]
 dyn_chain_04 [{shapeType=Capsule, length=2.8869, radius=1.1238, shapeName=dyn_chain_04_capsule, center=<2.9 0 0>, rotation=<rot 90 0 0>}]
 dyn_chain_05 [{shapeType=Capsule, length=3.0131, radius=1.1341, shapeName=dyn_chain_05_capsule, center=<2.8 0 0>, rotation=<rot 90 0 0>}]
 dyn_chain_06 [{shapeType=Capsule, length=1.172, radius=1.0333, shapeName=dyn_chain_06_capsule, center=<0.5 0 0>, rotation=<rot 90 0 0>}]
 dyn_chain_07 [{shapeType=Capsule, length=5.7712, radius=4.1692, shapeName=dyn_chain_07_capsule, center=<4.2389 1.7455 -0.5845>, rotation=<rot 90 0 0>}]
 thigh_l [{shapeType=Capsule, length=40, radius=8, center=<-25 0 0>, rotation=<rot 90 0 0>}]
 thigh_r [{shapeType=Capsule, length=40, radius=8, center=<25 0 0>, rotation=<rot 90 0 0>}]
}
constraints [
 {bone1Name=dyn_chain_01, bone2Name=Melee_TwoHanded_Handle, swing1LimitDegrees=90, swing2LimitDegrees=55, twistLimitDegrees=5}
 {bone1Name=dyn_chain_02, bone2Name=dyn_chain_01, swing1LimitDegrees=45, swing2LimitDegrees=45, twistLimitDegrees=5}
 {bone1Name=dyn_chain_03, bone2Name=dyn_chain_02, swing1LimitDegrees=45, swing2LimitDegrees=45, twistLimitDegrees=4}
 {bone1Name=dyn_chain_04, bone2Name=dyn_chain_03, swing1LimitDegrees=45, swing2LimitDegrees=45, twistLimitDegrees=4}
 {bone1Name=dyn_chain_05, bone2Name=dyn_chain_04, swing1LimitDegrees=45, swing2LimitDegrees=45, twistLimitDegrees=4}
 {bone1Name=dyn_chain_06, bone2Name=dyn_chain_05, swing1LimitDegrees=10, swing2LimitDegrees=50, twistLimitDegrees=0}
 {bone1Name=dyn_chain_07, bone2Name=dyn_chain_06, swing1LimitDegrees=45, swing2LimitDegrees=9, twistLimitDegrees=0}
 {
  bone1Name=calf_l
  bone2Name=Melee_TwoHanded_Handle
  swing1LimitDegrees=45
  swing1Motion=Free
  swing2LimitDegrees=45
  swing2Motion=Free
  twistLimitDegrees=45
  twistMotion=Free
 }
 {
  bone1Name=calf_r
  bone2Name=Melee_TwoHanded_Handle
  swing1LimitDegrees=45
  swing1Motion=Free
  swing2LimitDegrees=45
  swing2Motion=Free
  twistLimitDegrees=45
  twistMotion=Free
 }
 {
  bone1Name=thigh_l
  bone2Name=Melee_TwoHanded_Handle
  swing1LimitDegrees=45
  swing1Motion=Free
  swing2LimitDegrees=45
  swing2Motion=Free
  twistLimitDegrees=45
  twistMotion=Free
 }
 {
  bone1Name=thigh_r
  bone2Name=Melee_TwoHanded_Handle
  swing1LimitDegrees=45
  swing1Motion=Free
  swing2LimitDegrees=45
  swing2Motion=Free
  twistLimitDegrees=45
  twistMotion=Free
 }
]
